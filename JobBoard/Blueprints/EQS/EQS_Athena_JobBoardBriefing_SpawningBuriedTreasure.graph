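# EQS query that picks a location to spawn buried treasure for an Athena job board briefing.
# A spiral donut around the generic spawning context, with scalable-float radii, projects points onto navigation excluding water.
# Tests in fixed order check projection, overlaps, water, safe zone and buildings, then favor points nearer the safe zone center and farther from the querier.
# Two surface tests check flatness and soft ground materials like dirt, sand and snow.
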EnvQuery /JobBoard/Blueprints/EQS/EQS_Athena_JobBoardBriefing_SpawningBuriedTreasure
options [{
  generator {
   name=EnvQueryGenerator_Donut_0
   class=EnvQueryGenerator_Donut
   properties {
    bAutoSortTests=false
    bUseSpiralPattern=true
    center=&/EQSC_Athena_JobBoardBriefing_SpawningGeneric.EQSC_Athena_JobBoardBriefing_SpawningGeneric_C
    numberOfRings=4
    pointsPerRing=12
    arcDirection {dirMode=Rotation}
    innerRadius {
     dataBinding=&self:EnvQueryGenerator_Donut_0.FortAIDataProvider_ScalableFloat_0
     dataField=FloatValue
     defaultValue=15000
    }
    outerRadius {
     dataBinding=&self:EnvQueryGenerator_Donut_0.FortAIDataProvider_ScalableFloat_1
     dataField=FloatValue
     defaultValue=45000
    }
    projectionData {
     extentX=0
     navigationFilter=&NavFilter_NPC_ExcludeAllWater_C
     postProjectionVerticalOffset=-15
     projectDown=22000
    }
   }
  }
  tests [
   EnvQueryTest_Project_0 {
    class=EnvQueryTest_Project
    properties {
     scoringFactor=1
     testComment=""
     testPurpose=Filter
     projectionData {
      navigationFilter=null
      postProjectionVerticalOffset=0
      projectDown=200
      projectUp=0
      serializedChannel=ECC_GameTraceChannel6
      traceChannel=TraceTypeQuery7
      traceMode=GeometryByChannel
     }
    }
   }
   EnvQueryTest_Overlap_2 {
    class=EnvQueryTest_Overlap
    properties {testPurpose=FilterAndScore, overlapData {extentX=350, extentY=350, extentZ=350, overlapChannel=ECC_GameTraceChannel15, shapeOffset=<0 0 0>}}
   }
   FortQueryTest_InsideWater_0 {
    class=FortQueryTest_InsideWater
    properties {
     bDefineReferenceValue=false
     boolValue=false
     clampMaxType=None
     clampMinType=None
     filterType=Match
     floatValueMax=0
     floatValueMin=0
     multipleContextFilterOp=AllPass
     multipleContextScoreOp=AverageScore
     normalizationType=Absolute
     referenceValue=0
     scoreClampMax=0
     scoreClampMin=0
     scoringEquation=Constant
     scoringFactor=10
     testComment=""
     testPurpose=Filter
     testRadius=512
    }
   }
   EnvQueryTest_Volume_1 {
    class=EnvQueryTest_Volume
    properties {boolValue=false, testPurpose=Filter, volumeClass=/Script/FortniteGame.WaterVolume_Athena, volumeContext=null}
   }
   FortQueryTest_InsideAthenaSafeZone_0 {class=FortQueryTest_InsideAthenaSafeZone, properties {testPurpose=Filter}}
   FortQueryTest_InsideBuilding_1 {class=FortQueryTest_InsideBuilding, properties {testPurpose=Filter}}
   EnvQueryTest_Overlap_1 {
    class=EnvQueryTest_Overlap
    properties {overlapData {extentX=150, extentY=150, extentZ=150, overlapChannel=ECC_WorldStatic, shapeOffset=<0 0 200>}}
   }
   EnvQueryTest_Distance_0 {
    class=EnvQueryTest_Distance
    properties {
     distanceTo=/Script/FortniteAIServer.FortQueryContext_AthenaCurrentSafeZoneIndicatorCenter
     filterType=Range
     floatValueMin=0
     scoringEquation=InverseLinear
     scoringFactor=4
     testComment=""
     testMode=Distance2D
     testPurpose=Score
    }
   }
   EnvQueryTest_Distance_2 {
    class=EnvQueryTest_Distance
    properties {
     distanceTo=&EnvQueryContext_Querier
     filterType=Range
     floatValueMax=5000
     floatValueMin=1
     scoringEquation=Linear
     scoringFactor=1
     testComment=""
     testMode=Distance2D
     testPurpose=Score
    }
   }
   FortQueryTest_ValidSurface_0 {
    class=FortQueryTest_ValidSurface
    properties {flatSurfaceToleranceZ=20, radius=50, scoringFactor=2, traceOffsetDown=500, traceOffsetUp=100, surfaceMaterials []}
   }
   FortQueryTest_ValidSurface_3 {
    class=FortQueryTest_ValidSurface
    properties {
     bAreSurfaceMaterialsValid=true
     flatSurfaceToleranceZ=0
     radius=0
     scoringFactor=2
     traceCollisionChannel=ECC_WorldDynamic
     traceOffsetDown=2000
     traceOffsetUp=100
     surfaceMaterials [
      /Game/FortressPhysicalMaterials/Snow
      /Game/FortressPhysicalMaterials/Dirt
      /Game/FortressPhysicalMaterials/Grass
      /Game/FortressPhysicalMaterials/Sand
      /Game/Balance/Physics/Materials/PM_Mud
     ]
    }
   }
  ]
 }]
