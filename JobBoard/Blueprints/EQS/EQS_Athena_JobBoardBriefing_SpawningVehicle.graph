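# EQS query that picks a spot to spawn a vehicle for an Athena Job Board briefing.
# A spiral donut of points around the generic briefing spawn context is ground-projected by downward trace, then tested for overlaps, water, nearby buildings and flat dirt, asphalt or stone surfaces.
# Distance scoring is weighted strongly toward the current safe zone indicator center, with a smaller term for distance from the querier.
# Test auto-sorting is off, so tests keep their authored order.
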EnvQuery /JobBoard/Blueprints/EQS/EQS_Athena_JobBoardBriefing_SpawningVehicle
options [{
  generator {
   name=EnvQueryGenerator_Donut_1
   class=EnvQueryGenerator_Donut
   properties {
    bAutoSortTests=false
    bUseSpiralPattern=true
    center=&/EQSC_Athena_JobBoardBriefing_SpawningGeneric.EQSC_Athena_JobBoardBriefing_SpawningGeneric_C
    numberOfRings=6
    pointsPerRing=10
    arcAngle {dataBinding=null, dataField=FloatValue, defaultValue=360}
    arcDirection {dirMode=Rotation}
    innerRadius {dataBinding=null, dataField=FloatValue, defaultValue=4000}
    outerRadius {dataBinding=null, dataField=FloatValue, defaultValue=12000}
    projectionData {extentX=0, navigationFilter=null, postProjectionVerticalOffset=0, traceMode=None}
   }
  }
  tests [
   EnvQueryTest_Project_0 {
    class=EnvQueryTest_Project
    properties {
     scoringFactor=1
     testComment=""
     testPurpose=Filter
     projectionData {
      navigationFilter=null
      postProjectionVerticalOffset=0
      projectDown=22000
      projectUp=0
      serializedChannel=ECC_GameTraceChannel6
      traceChannel=TraceTypeQuery7
      traceMode=GeometryByChannel
     }
    }
   }
   EnvQueryTest_Overlap_1 {
    class=EnvQueryTest_Overlap
    properties {overlapData {extentX=290, extentY=290, extentZ=290, overlapChannel=ECC_GameTraceChannel1, shapeOffset=<0 0 300>}}
   }
   EnvQueryTest_Overlap_2 {
    class=EnvQueryTest_Overlap
    properties {testPurpose=FilterAndScore, overlapData {extentX=550, extentY=550, extentZ=550, overlapChannel=ECC_GameTraceChannel15, shapeOffset=<0 0 0>}}
   }
   FortQueryTest_InsideWater_0 {
    class=FortQueryTest_InsideWater
    properties {
     bDefineReferenceValue=false
     boolValue=false
     clampMaxType=None
     clampMinType=None
     filterType=Match
     floatValueMax=0
     floatValueMin=0
     multipleContextFilterOp=AllPass
     multipleContextScoreOp=AverageScore
     normalizationType=Absolute
     referenceValue=0
     scoreClampMax=0
     scoreClampMin=0
     scoringEquation=Constant
     scoringFactor=10
     testComment=""
     testPurpose=Filter
     testRadius=512
    }
   }
   FortQueryTest_HasNearbyBuildings_0 {class=FortQueryTest_HasNearbyBuildings, properties {extentXY=2, scoringFactor=0.5, testPurpose=FilterAndScore}}
   EnvQueryTest_Distance_1 {
    class=EnvQueryTest_Distance
    properties {
     distanceTo=/Script/FortniteAIServer.FortQueryContext_AthenaCurrentSafeZoneIndicatorCenter
     filterType=Range
     floatValueMin=0
     scoringEquation=Linear
     scoringFactor=-5
     testComment=""
     testMode=Distance2D
     testPurpose=Score
    }
   }
   EnvQueryTest_Distance_0 {
    class=EnvQueryTest_Distance
    properties {
     distanceTo=&EnvQueryContext_Querier
     filterType=Range
     floatValueMax=5000
     floatValueMin=1
     scoringEquation=Linear
     testComment=""
     testMode=Distance2D
     testPurpose=Score
     scoringFactor {dataBinding=null, dataField=FloatValue, defaultValue=0.5}
    }
   }
   FortQueryTest_ValidSurface_1 {
    class=FortQueryTest_ValidSurface
    properties {
     bAreSurfaceMaterialsValid=true
     flatSurfaceToleranceZ=20
     multipleContextFilterOp=AnyPass
     radius=50
     testPurpose=FilterAndScore
     traceOffsetDown=100
     traceOffsetUp=100
     surfaceMaterials [/Game/FortressPhysicalMaterials/Dirt, /Game/FortressPhysicalMaterials/Asphalt, /Game/FortressPhysicalMaterials/Stone]
    }
   }
  ]
 }]
